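# Translucent unlit, two-sided Niagara sprite material for a pickaxe claw impact slash.
# The squared red channel of the claw texture, offset by dynamic GradientProgress, blends body and tip colors, tinted by particle color.
# useGlow blends to plain particle color and switches opacity from red to blue channel.
# Dynamic Dissolve drives a HeightLerp over cloud noise to erode opacity, with both textures sampled on axis-swapped UVs.
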Material /Game/Effects/Niagara/MasterMaterials/Claw/M_Pickaxe_ClawImpact
graph {
 properties {
  bUsedWithNiagaraSprites=true
  blendMode=BLEND_Translucent
  pixelDepthOffsetMode=PDOM_Legacy
  shadingModel=MSM_Unlit
  twoSided=true
  usedShadingModels=Unlit
 }
 nodes [
  n1 = ParticleColor()
  n2 = TextureSampleParameter2D(UVs=MaterialFunctionCall(XY=TextureCoordinate()) {materialFunction=&Swizzle}.YX) {parameterName=ClawTexture, texture=&/T_Pickaxe_ClawSlash}
  n3 = Multiply(A=@n2.R, B=@n2.R)
  n4 = DynamicParameter() {defaultValue=<rgba 1 -1 1 1>, paramNames [Dissolve, GradientProgress, Param3, Param4]}
  n5 = LinearInterpolate(
   A=VectorParameter() {parameterName=LerpColorBody, defaultValue=<rgba 0.0016 0.5 0.2146 1>}.RGB
   B=VectorParameter() {parameterName=LerpColorTIp, defaultValue=<rgba 0 4 0.7026 0.5>}.RGB
   Alpha=Saturate(None=Subtract(A=@n3, B=@n4.GradientProgress))
  )
  MaterialFunctionCall(Emissive=LinearInterpolate(A=Multiply(A=@n1.RGB, B=@n5), B=@n1.RGB, Alpha=ScalarParameter() {parameterName=useGlow})) {materialFunction=&MF_FX_EmissiveAdjustments}
  n6 = Multiply(A=@n2.R) {constB=4}
  n7 = ScalarParameter() {parameterName=useGlow}
  n8 = MaterialFunctionCall(XY=TextureCoordinate() {vTiling=0.3}) {materialFunction=&Swizzle}
  n9 = Multiply(A=MaterialFunctionCall(
    Transition Phase=DynamicParameter() {defaultValue=<rgba 1 -1 1 1>, paramNames [Dissolve, GradientProgress, Param3, Param4]}.Dissolve
    Height Texture=TextureSample(UVs=@n8.YX) {samplerType=SAMPLERTYPE_LinearGrayscale, texture=/Game/Athena/Prototype/Terrain/T_CloudCoverage}.R
   ) {materialFunction=&HeightLerp}.Alpha) {constB=3}
  Multiply(A=@n1.A, B=Multiply(A=LinearInterpolate(A=@n6, B=@n2.B, Alpha=@n7), B=Saturate(None=@n9)))
 ]
}
